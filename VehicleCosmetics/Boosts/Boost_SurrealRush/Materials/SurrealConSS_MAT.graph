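# Unlit masked Niagara particle material for the Surreal Rush vehicle boost cone effect, usable on sprites, ribbons and meshes.
# The "Painted" dynamic parameter blends between a particle-color-tinted black-and-white cone texture and a slightly brightened second cone texture in its own colors.
# The blend feeds the shared emissive adjustments function, while particle alpha times the second texture's alpha forms the mask.
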Material /VehicleCosmetics/Boosts/Boost_SurrealRush/Materials/SurrealConSS_MAT
graph {
 properties {
  bAutomaticallySetUsageInEditor=false
  bUsedWithNiagaraMeshParticles=true
  bUsedWithNiagaraRibbons=true
  bUsedWithNiagaraSprites=true
  blendMode=BLEND_Masked
  pixelDepthOffsetMode=PDOM_Legacy
  shadingModel=MSM_Unlit
  usedShadingModels=Unlit
 }
 nodes [
  MaterialExpressionVertexColor_1 = ParticleColor()
  n1 = TextureSample() {texture=/VehicleCosmetics/Boosts/Boost_SurrealRush/Textures/T_VCVFX_Boost_SurrealRush_Cone04BW_SS}
  n2 = TextureSample() {texture=/VehicleCosmetics/Boosts/Boost_SurrealRush/Textures/T_VCVFX_Boost_SurrealRush_Cone05_SS}
  n3 = Constant() {r=1.2}
  MaterialFunctionCall(Emissive=LinearInterpolate(
    A=Multiply(A=@MaterialExpressionVertexColor_1.RGB, B=@n1.RGB)
    B=Multiply(A=@n2.RGB, B=@n3)
    Alpha=DynamicParameter() {paramNames [Painted,Param2,Param3,Param4]}.Painted
   )) {materialFunction=&MF_FX_EmissiveAdjustments}
  Multiply(A=@MaterialExpressionVertexColor_1.A, B=@n2.A)
 ]
}
